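# Masked unlit base material for energy-type flying damage text, drawing number glyphs or a plus/minus sign.
# UV1 samples the damage text atlas, while half-tiled V coordinates offset by the "+Or-" parameter select the sign.
# The red channel of UV2 blends glyph and sign for both the green-channel shape and the normal map.
# The blended normal is Z-scaled, normalized and fed into the shared DamageNumberWPO function.
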Material /Game/UI/DamageText/Energy/FlyingTextMat_EnergyBase
graph {
 properties {blendMode=BLEND_Masked, numCustomizedUVs=5, pixelDepthOffsetMode=PDOM_Legacy, shadingModel=MSM_Unlit, usedShadingModels=Unlit}
 nodes [
  Constant()
  n1 = TextureCoordinate() {coordinateIndex=1}
  n2 = Add(A=TextureCoordinate() {vTiling=0.5}, B=AppendVector(A=Constant(), B=ScalarParameter() {parameterName=+Or-}))
  n3 = ComponentMask(None=TextureCoordinate() {coordinateIndex=2}) {r=true}
  LinearInterpolate(
   A=TextureSample(UVs=@n1) {samplerType=SAMPLERTYPE_LinearColor, texture=/Game/UI/DamageText/Default/damage_text}.G
   B=TextureSample(UVs=@n2) {texture=/Game/UI/DamageText/Default/+-Diffuse.+-Diffuse}.G
   Alpha=@n3
  )
  n4 = LinearInterpolate(
   A=TextureSample(UVs=@n1) {samplerType=SAMPLERTYPE_Normal, texture=/Game/UI/DamageText/Default/damage_text_Normals}.RGB
   B=TextureSample(UVs=@n2) {samplerType=SAMPLERTYPE_Normal, texture=&Default/+-Normal.+-Normal}.RGB
   Alpha=@n3
  )
  n5 = Constant3Vector() {constant=<rgba 1 1 0.5 0>}
  MaterialFunctionCall!(
   Normals=Normalize(VectorInput=Multiply(A=@n4, B=@n5))
   Debug texture coordinate input
   Animation
   Additonal Scale
   "Additional Camera Offset "
   "Additonal Screenspace XY offset "
  ) {materialFunction=/Game/UI/DamageText/Base/DamageNumberWPO}
 ]
}
