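# Unlit translucent material for the energy glow on flying damage text, reusing the Fire damage-number textures.
# Panning noise over radial coordinates slightly distorts lookups into the digit and plus/minus textures, which UV2 red blends between.
# The amplified, clamped red channel forms the mask and drives the glow color, scaled by time of day and a letter highlight.
# DamageNumberWPO supplies that highlight and takes scale and camera-offset inputs.
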Material /Game/UI/DamageText/Energy/FlyingTextMat_EnergyGlow
graph {
 properties {blendMode=BLEND_Translucent, numCustomizedUVs=5, pixelDepthOffsetMode=PDOM_Legacy, shadingModel=MSM_Unlit, usedShadingModels=Unlit}
 nodes [
  n1 = Constant() {r=1}
  n2 = Constant() {r=0.615}
  n3 = VectorParameter() {parameterName=Energy_Color_Glow, defaultValue=<rgba 0 85.3244 200 1>}
  n4 = TextureCoordinate() {coordinateIndex=1, uTiling=0.46, vTiling=0.2476}
  n5 = ComponentMask(None=TextureSample(UVs=Panner(Coordinate=@n4) {speedX=0.4, speedY=0.4}) {samplerType=SAMPLERTYPE_LinearColor, texture=/Game/Global/Textures/Default/Noise/TilingNoise06_N_uncompressed}.RGB) {g=true, r=true}
  n6 = MaterialFunctionCall() {materialFunction=&VectorToRadialValue}
  n7 = Add(
   A=Multiply(A=Subtract(A=@n5) {constB=0.5}) {constB=0.1}
   B=Panner(Coordinate=@n6.Radial Coordinates, Time=Time()) {speedX=-0.2301, speedY=0.0354}
  ) {constB=0}
  n8 = ScalarParameter() {defaultValue=0.0063, parameterName=Distortion_Control}
  n9 = LinearInterpolate(
   A=TextureCoordinate() {coordinateIndex=1}
   B=ComponentMask(None=TextureSample(UVs=@n7) {samplerType=SAMPLERTYPE_Normal, texture=/Game/Global/Textures/Default/Noise/64x64Noise}.RGB) {g=true, r=true}
   Alpha=@n8
  )
  n10 = TextureCoordinate()
  n11 = TextureSample(UVs=Panner(Coordinate=TextureCoordinate() {uTiling=0.46, vTiling=0.2476}) {speedX=0.4, speedY=0.4}) {samplerType=SAMPLERTYPE_LinearColor, texture=&TilingNoise06_N_uncompressed}
  n12 = Multiply(A=Subtract(A=ComponentMask(None=@n11.RGB) {g=true, r=true}) {constB=0.5}) {constB=0.1}
  n13 = Panner(Coordinate=@n6.Radial Coordinates, Time=Time()) {speedX=-0.2301, speedY=0.0354}
  n14 = ComponentMask(None=TextureSample(UVs=Add(A=@n12, B=@n13) {constB=0}) {samplerType=SAMPLERTYPE_Normal, texture=&64x64Noise}.RGB) {g=true, r=true}
  n15 = Constant()
  n16 = ScalarParameter() {parameterName=+Or-_1}
  n17 = TextureSample(UVs=Add(A=LinearInterpolate(A=@n10, B=@n14, Alpha=@n8), B=AppendVector(A=@n15, B=@n16))) {texture=/Game/UI/DamageText/Fire/DamageNumbers_Fire_R_brightEdges_G_Crevasses_B_AO_+-.DamageNumbers_Fire_R_brightEdges_G_Crevasses_B_AO_+-}
  n18 = TextureCoordinate() {coordinateIndex=2}
  n19 = MaterialFunctionCall(Float3=LinearInterpolate(
    A=TextureSample(UVs=@n9) {texture=/Game/UI/DamageText/Fire/DamageNumbers_Fire_R_brightEdges_G_Crevasses_B_AO}.R
    B=@n17.R
    Alpha=ComponentMask(None=@n18) {r=true}
   )) {desc=R opacity mask G Distance field, materialFunction=&BreakOutFloat3Components}
  n20 = Clamp(None=Multiply(A=@n19.R, B=ScalarParameter() {defaultValue=25}))
  Multiply(
   A=Multiply(
    A=MaterialFunctionCall(Color_Morning=@n1, Color_Day=@n1, Color_Evening=@n1, Color_Night=@n2) {materialFunction=&Color_Time_of_Day}.Result
    B=LinearInterpolate(B=@n3.RGB, Alpha=@n20)
   )
   B=MaterialFunctionCall!(
    Normals
    Debug texture coordinate input
    Animation
    Additonal Scale=Constant2Vector() {g=1.1, r=1.3}
    "Additional Camera Offset "=ScalarParameter() {defaultValue=-5, parameterName=Camera Offset}
    "Additonal Screenspace XY offset "
   ) {materialFunction=/Game/UI/DamageText/Base/DamageNumberWPO}.Pixel shader Letter Highlight - uv 3
  )
  Reroute(None=@n20)
 ]
}
