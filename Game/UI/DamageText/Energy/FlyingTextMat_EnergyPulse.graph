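# Additive unlit material for flying damage numbers with an animated energy pulse look.
# Glyph shapes blend digit and +/- distance-field textures via a UV2 mask, feeding a depth fade.
# A panning lightning gradient, distorted by time-stepped noise, drives cycling height-lerped pulses tinted by a color parameter.
# DamageNumberWPO handles scale and camera offset, supplying the letter highlight and lifetime fade that multiply the time-of-day-adjusted glow.
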Material /Game/UI/DamageText/Energy/FlyingTextMat_EnergyPulse
graph {
 properties {blendMode=BLEND_Additive, numCustomizedUVs=6, pixelDepthOffsetMode=PDOM_Legacy, shadingModel=MSM_Unlit, usedShadingModels=Unlit}
 nodes [
  n1 = TextureCoordinate() {coordinateIndex=1}
  n2 = Add(A=TextureCoordinate() {vTiling=0.5}, B=AppendVector(A=Constant(), B=ScalarParameter() {parameterName=+Or-}))
  n3 = TextureCoordinate() {coordinateIndex=2}
  n4 = MaterialFunctionCall(Float3=LinearInterpolate(
    A=TextureSample(UVs=@n1) {texture=/Game/UI/DamageText/Fire/FireNumbers_opacity_g_distancefield}.G
    B=TextureSample(UVs=@n2) {texture=/Game/UI/DamageText/Fire/FireNumbers_opacity_g_distancefield_+-.FireNumbers_opacity_g_distancefield_+-}.RGB
    Alpha=ComponentMask(None=@n3) {r=true}
   )) {desc=R opacity mask G Distance field, materialFunction=&BreakOutFloat3Components}
  DepthFade(Opacity=@n4.G, FadeDistance=ScalarParameter() {parameterName=Distance Fade}) {fadeDistanceDefault=50}
  n5 = Constant() {r=1}
  n6 = MaterialFunctionCall(Color_Morning=@n5, Color_Day=@n5, Color_Evening=@n5, Color_Night=Constant() {r=0.615}) {materialFunction=&Color_Time_of_Day}
  n7 = Multiply(A=Time()) {constB=16}
  n8 = Panner(Coordinate=TextureCoordinate() {uTiling=0.46, vTiling=0.2476}, Time=Divide(A=Ceil(None=@n7)) {constB=4}) {speedX=0.4, speedY=0.4}
  n9 = Subtract(A=ComponentMask(None=TextureSample(UVs=@n8) {samplerType=SAMPLERTYPE_LinearColor, texture=/Game/Global/Textures/Default/Noise/TilingNoise06_N_uncompressed}.RGB) {g=true, r=true}) {constB=0.5}
  n10 = Add(A=TextureCoordinate() {coordinateIndex=5}, B=TextureCoordinate())
  n11 = Time()
  n12 = TextureSample(UVs=Add(A=Multiply(A=@n9) {constB=0.1}, B=Panner(Coordinate=@n10, Time=@n11) {speedX=-0.2301, speedY=0.0354}) {constB=0}) {samplerType=SAMPLERTYPE_LinearColor, texture=/Game/Global/Textures/Default/Masks_and_Gradients/TestLightningGradient}
  n13 = Frac(None=Add(A=Time(), B=@n12.B))
  n14 = SphereMask(A=@n12.G, B=Multiply(A=Subtract(A=@n13) {constB=0.2}) {constB=10}) {attenuationRadius=1, desc="! 50 perc", hardnessPercent=57.9646}
  n15 = OneMinus(None=Clamp(None=Multiply(A=@n14) {constB=20}))
  n16 = OneMinus(None=@n12.A)
  n17 = Multiply(
   A=Multiply(
    A=@n12.R
    B=MaterialFunctionCall(Transition Phase=@n15, Height Texture=@n16) {materialFunction=&HeightLerp}.Results
   )
   B=VectorParameter() {parameterName=Energy_Color_Pulse, defaultValue=<rgba 20 1.5625 9.8926 1>}.RGB
  )
  n18 = MaterialFunctionCall!(
   Normals
   Debug texture coordinate input
   Animation
   Additonal Scale=Multiply(A=Constant2Vector() {g=1.2221, r=1.4442}, B=ScalarParameter() {defaultValue=0.9366, parameterName=Additional Scale}) {constB=1.1115}
   "Additional Camera Offset "=ScalarParameter() {defaultValue=10, parameterName=Camera Offset}
   "Additonal Screenspace XY offset "
  ) {materialFunction=/Game/UI/DamageText/Base/DamageNumberWPO}
  Multiply(A=Multiply(A=Multiply(A=@n6.Result, B=@n17), B=@n18.Pixel shader Letter Highlight - uv 3) {desc=eMISSIVE}, B=OneMinus(None=@n18.Time))
 ]
}
